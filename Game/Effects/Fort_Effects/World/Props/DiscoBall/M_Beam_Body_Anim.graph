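# Nanite-enabled lit material for an animated beam light body prop in the disco ball effects, using concert stage diffuse, normal and mask textures.
# Object pivot position seeds per-instance offsets for time-driven sines.
# These sines pick colors from a swatch texture for a sphere-masked emissive spot and swing the mesh about its local X axis.
# The rotation offset fades out with camera distance.
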Material /Game/Effects/Fort_Effects/World/Props/DiscoBall/M_Beam_Body_Anim
graph {
 properties {bAutomaticallySetUsageInEditor=false, bUsedWithNanite=true, usedShadingModels=Default Lit}
 nodes [
  n1 = MaterialFunctionCall(Float3=MaterialFunctionCall() {materialFunction=&ObjectPivotPoint}.Object Pivot Location) {materialFunction=&BreakOutFloat3Components}
  n2 = Add(A=@n1.R, B=@n1.R)
  n3 = Frac(None=Divide(A=Add(A=@n2, B=@n1.B)) {constB=10})
  n4 = Multiply(A=@n3) {constB=33.3}
  n5 = Time()
  n6 = Sine(None=Multiply(A=Add(A=@n4, B=@n5), B=ScalarParameter() {defaultValue=0.112, parameterName=SpeedColor1}))
  Divide(A=Add(A=@n6))
  n7 = Time()
  n8 = Sine(None=Multiply(A=Add(A=@n3, B=@n7), B=ScalarParameter() {defaultValue=0.16, parameterName=SpeedColor2}))
  Divide(A=Add(A=@n8))
  n9 = AppendVector(A=@n6, B=@n8)
  n10 = Add(
   A=TextureCoordinate()
   B=ComponentMask(None=VectorParameter() {parameterName=EmissivePosition, defaultValue=<rgba 0.157 0.24 0 1>}.RGB) {g=true, r=true}
  )
  n11 = ComponentMask(None=VectorParameter() {parameterName=EmissiveScale, defaultValue=<rgba 0.328 0.096 0 1>}.RGB) {g=true, r=true}
  n12 = Power(
   Base=SphereMask(
    A=MaterialFunctionCall(UVs=@n10, Texture Scale=@n11) {materialFunction=&ScaleUVsByCenter}.UVs
    B=Constant2Vector() {g=0.5, r=0.5}
    Radius=ScalarParameter() {defaultValue=0.1011, parameterName=RadiusEmissive}
    Hardness=ScalarParameter() {defaultValue=0.25, parameterName=HardnessEmissive}
   )
   Exp=ScalarParameter() {defaultValue=3, parameterName=PowerEmissive}
  )
  n13 = ScalarParameter() {defaultValue=3, parameterName=BoostEmissive}
  MaterialFunctionCall(Emissive=Multiply(A=TextureSample(UVs=@n9) {texture=/Game/Environments/Sets/Concert/Textures/BeamColorSwatch}.RGB, B=Multiply(A=@n12, B=@n13))) {materialFunction=&MF_FX_EmissiveAdjustments}
  n14 = Transform(None=Constant3Vector() {constant=<rgba 1 0 0 0>}) {transformSourceType=TRANSFORMSOURCE_Local}
  n15 = Time()
  n16 = Sine(None=Multiply(A=Add(A=@n4, B=@n15), B=ScalarParameter() {defaultValue=0.3, parameterName=RotationSpeed}))
  n17 = ScalarParameter() {defaultValue=0.05, parameterName=RotationAngle}
  n18 = MaterialFunctionCall() {materialFunction=&ObjectPivotPoint}
  n19 = CameraPositionWS()
  n20 = Divide(
   A=Subtract(A=Distance(A=@n18.Object Pivot Location, B=@n19), B=ScalarParameter() {defaultValue=4000, parameterName=WPODisableOffset})
   B=ScalarParameter() {defaultValue=4000, parameterName=WPODisableBlend}
  )
  Multiply(
   A=RotateAboutAxis(
    NormalizedRotationAxis=Normalize(VectorInput=@n14)
    RotationAngle=Multiply(A=@n16, B=@n17)
    PivotPoint=@n18.Object Pivot Location
    Position=WorldPosition().XYZ
   )
   B=OneMinus(None=Saturate(None=@n20))
  )
  TextureSample() {samplerType=SAMPLERTYPE_Normal, texture=/Game/Environments/Sets/Concert/Textures/T_Stage_N}
  TextureSample() {texture=/Game/Environments/Sets/Concert/Textures/T_Stage_D}
  TextureSample() {samplerType=SAMPLERTYPE_Masks, texture=/Game/Environments/Sets/Concert/Textures/T_Stage_S}
 ]
}
