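# Unlit additive particle material for disco ball light beams, built for mesh particles.
# A SubUV color swatch feeds an FX emissive-adjustment function.
# A ramp texture's green channel times particle alpha feeds a depth fade that softens intersections.
# A separate world-aligned smoke texture scrolls by adding time to world Z, with no visible link to the color or fade branches.
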Material /Game/Effects/Fort_Effects/World/Props/DiscoBall/M_Beam_Light
graph {
 properties {bUsedWithMeshParticles=true, blendMode=BLEND_Additive, pixelDepthOffsetMode=PDOM_Legacy, shadingModel=MSM_Unlit, usedShadingModels=Unlit}
 nodes [
  DepthFade(Opacity=Multiply(A=TextureSample() {texture=/Game/Effects/Fort_Effects/Textures/Ramps/T_Ramp_Limbo_Pool_Gas}.G, B=ParticleColor().A))
  MaterialFunctionCall(Emissive=Multiply(A=ParticleSubUV!(UVs, Tex, Apply View MipBias) {texture=/Game/Environments/Sets/Concert/Textures/BeamColorSwatch}.RGB)) {materialFunction=&MF_FX_EmissiveAdjustments}
  n1 = Constant() {r=128}
  n2 = WorldPosition()
  n3 = MaterialFunctionCall(Speed=@n1) {materialFunction=&TimeWithSpeedVariable}
  n4 = ComponentMask(None=@n2.XYZ) {b=true}
  MaterialFunctionCall(
   TextureObject=TextureObject() {texture=/Game/Effects/Fort_Effects/Textures/T_Limbo_smoke_pool}
   TextureSize=@n1
   WorldPosition=AppendVector(A=ComponentMask(None=@n2.XYZ) {g=true, r=true}, B=Add(A=@n3.Time, B=@n4))
  ) {materialFunction=&WorldAlignedTexture}
 ]
}
