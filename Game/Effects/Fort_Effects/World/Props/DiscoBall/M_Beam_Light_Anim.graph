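# Additive unlit material for disco ball light beams that shift color and sway over time.
# A seed from the object pivot position offsets time-driven sines, two of which form UVs into a color swatch for emissive.
# Another sine drives world position offset rotation around the pivot on local X, faded out by camera distance.
# Opacity combines a ramp texture and particle alpha through depth fade.
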Material /Game/Effects/Fort_Effects/World/Props/DiscoBall/M_Beam_Light_Anim
graph {
 properties {blendMode=BLEND_Additive, pixelDepthOffsetMode=PDOM_Legacy, shadingModel=MSM_Unlit, usedShadingModels=Unlit}
 nodes [
  DepthFade(Opacity=Multiply(A=TextureSample() {texture=/Game/Effects/Fort_Effects/Textures/Ramps/T_Ramp_Limbo_Pool_Gas}.G, B=ParticleColor().A))
  n1 = MaterialFunctionCall(Float3=MaterialFunctionCall() {materialFunction=&ObjectPivotPoint}.Object Pivot Location) {materialFunction=&BreakOutFloat3Components}
  n2 = Add(A=@n1.R, B=@n1.R)
  n3 = Frac(None=Divide(A=Add(A=@n2, B=@n1.B)) {constB=10})
  n4 = Multiply(A=@n3) {constB=33.3}
  n5 = Time()
  n6 = Sine(None=Multiply(A=Add(A=@n4, B=@n5), B=ScalarParameter() {defaultValue=0.112, parameterName=SpeedColor1}))
  Divide(A=Add(A=@n6))
  n7 = Time()
  n8 = Sine(None=Multiply(A=Add(A=@n3, B=@n7), B=ScalarParameter() {defaultValue=0.16, parameterName=SpeedColor2}))
  Divide(A=Add(A=@n8))
  n9 = AppendVector(A=@n6, B=@n8)
  MaterialFunctionCall(Emissive=Multiply(A=TextureSample(UVs=@n9) {texture=/Game/Environments/Sets/Concert/Textures/BeamColorSwatch}.RGB)) {materialFunction=&MF_FX_EmissiveAdjustments}
  n10 = Transform(None=Constant3Vector() {constant=<rgba 1 0 0 0>}) {transformSourceType=TRANSFORMSOURCE_Local}
  n11 = Time()
  n12 = Sine(None=Multiply(A=Add(A=@n4, B=@n11), B=ScalarParameter() {defaultValue=0.3, parameterName=RotationSpeed}))
  n13 = ScalarParameter() {defaultValue=0.05, parameterName=RotationAngle}
  n14 = MaterialFunctionCall() {materialFunction=&ObjectPivotPoint}
  n15 = CameraPositionWS()
  n16 = Divide(
   A=Subtract(A=Distance(A=@n14.Object Pivot Location, B=@n15), B=ScalarParameter() {defaultValue=4000, parameterName=WPODisableOffset})
   B=ScalarParameter() {defaultValue=4000, parameterName=WPODisableBlend}
  )
  Multiply(
   A=RotateAboutAxis(
    NormalizedRotationAxis=Normalize(VectorInput=@n10)
    RotationAngle=Multiply(A=@n12, B=@n13)
    PivotPoint=@n14.Object Pivot Location
    Position=WorldPosition().XYZ
   )
   B=OneMinus(None=Saturate(None=@n16))
  )
 ]
}
