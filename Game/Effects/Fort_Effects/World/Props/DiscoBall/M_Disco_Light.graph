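# Unlit additive material for disco ball light effects on mesh particles, including Niagara.
# Particle color passes through MF_FX_EmissiveAdjustments, while a ramp texture's green channel times particle alpha feeds a depth fade.
# The static switch WorldUVMark_Z optionally multiplies that fade by a world-space vertical gradient, which rises below the offset particle position and is shaped by divide, power and multiply parameters.
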Material /Game/Effects/Fort_Effects/World/Props/DiscoBall/M_Disco_Light
graph {
 properties {
  bUsedWithMeshParticles=true
  bUsedWithNiagaraMeshParticles=true
  blendMode=BLEND_Additive
  pixelDepthOffsetMode=PDOM_Legacy
  shadingModel=MSM_Unlit
  usedShadingModels=Unlit
 }
 nodes [
  n1 = ParticleColor()
  MaterialFunctionCall(Emissive=@n1.RGB) {materialFunction=&MF_FX_EmissiveAdjustments}
  n2 = DepthFade(Opacity=Multiply(A=TextureSample() {texture=/Game/Effects/Fort_Effects/Textures/Ramps/T_Ramp_Limbo_Pool_Gas}.G, B=@n1.A))
  n3 = ParticlePositionWS()
  n4 = VectorParameter() {parameterName=ParticlePositionOffset, defaultValue=<rgba 0 0 0 1>}
  n5 = Divide(A=Subtract(A=Add(A=@n3, B=@n4.RGB), B=WorldPosition().XYZ), B=ScalarParameter() {defaultValue=500, parameterName=WorldUV_Divide})
  n6 = Multiply(
   A=Power(Base=ComponentMask(None=@n5) {b=true}, Exp=ScalarParameter() {defaultValue=1, parameterName=WorldUV_Power})
   B=ScalarParameter() {defaultValue=1, parameterName=WorldUV_Multiply}
  )
  StaticSwitchParameter(True=Multiply(A=@n2, B=Saturate(None=@n6)), False=@n2) {parameterName=WorldUVMark_Z}
 ]
}
