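# Unlit additive material for soft spotlight beams, filed with the DiscoBall prop effects.
# Smoke noise pans slowly on the second UV channel and is multiplied by two samples of a gas ramp texture, one halved, then by a scalar parameter.
# An inverted Fresnel softens the beam edges, and DepthFade takes the product to soften intersections with geometry.
# A light-cone texture is sampled separately, with no connection shown.
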Material /Game/Effects/Fort_Effects/World/Props/DiscoBall/M_SpotLightBeams
graph {
 properties {blendMode=BLEND_Additive, pixelDepthOffsetMode=PDOM_Legacy, shadingModel=MSM_Unlit, usedShadingModels=Unlit}
 nodes [
  n1 = StaticBool() {value=true}
  n2 = Constant() {r=4}
  n3 = Panner(Coordinate=TextureCoordinate() {coordinateIndex=1}) {speedX=0.05, speedY=0.05}
  n4 = Multiply(
   A=Multiply(
    A=TextureSample(UVs=@n3) {texture=/Game/Global/Textures/Default/Tileable_Patterns/T_Smoke_Puff_Tile_Up}.RGB
    B=TextureSample() {texture=/Game/Effects/Fort_Effects/Textures/Ramps/T_Ramp_Limbo_Pool_Gas}.RGB
   )
   B=Multiply(A=TextureSample() {texture=&T_Ramp_Limbo_Pool_Gas}.RGB, B=Constant() {r=0.5})
  )
  n5 = ScalarParameter() {defaultValue=0.8}
  DepthFade(Opacity=Multiply(
    A=MaterialFunctionCall(Invert Fresnel=@n1, Power=@n2) {materialFunction=&Fresnel_Function}.Result
    B=Multiply(A=@n4, B=@n5)
   ))
  TextureSample() {texture=/Game/Environments/Sets/Concert/Textures/T_LightCones_D}
 ]
}
